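; BP_Outdoor_WoodCafeTableSet_B is an outdoor wooden cafe table set prop with plant and napkin holder components plus two instanced static mesh components.
; Its construction script only calls the parent, and the BeginPlay, overlap and Tick events are empty.
; The only logic of its own runs on Destroyed, which explicitly destroys those four components.
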(event EventBeginPlay)

(event Collision|EventActorBeginOverlap
  (params (OtherActor :type Actor%20Object%20Reference)))

(event EventTick
  (params (DeltaSeconds :type Float%20%28single-precision%29)))

(event EventDestroyed
  (Components|DestroyComponent (Variables|Default|GetInstancedStaticMesh))
  (Components|DestroyComponent (Variables|Default|GetSM_Outdoor_CafeTablePlant_A))
  (Components|DestroyComponent (Variables|Default|GetSM_Outdoor_CafeNapkinHolder_A))
  (Components|DestroyComponent (Variables|Default|GetInstancedStaticMesh1)))

(fn ConstructionScript
  (|CallParentFunction))
